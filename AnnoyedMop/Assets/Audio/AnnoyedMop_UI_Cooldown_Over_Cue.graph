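# A UI sound cue for the AnnoyedMop cooldown-over sound.
# A single wave player feeds a modulator with pitch fixed at unity, so playback pitch does not vary.
# It plays in the Fort_UI sound class at a reduced cue volume, with a very large max distance.
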SoundCue /AnnoyedMop/Assets/Audio/AnnoyedMop_UI_Cooldown_Over_Cue
root=Modulator_0
nodes [
 {id=Modulator_0, children [WavePlayer_0], properties {pitchMax=1, pitchMin=1}}
 {id=WavePlayer_0, properties {soundWaveAssetPtr=&/IP_AnnoyedMop_UI_Cooldown_Over}}
]
properties {maxDistance=2097152, soundClassObject=&Fort_UI, volumeMultiplier=0.5}
