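# Additive UI material that draws glowing cooldown bracket arcs with a progress arch for the ColdDay UI.
# Circle SDF outlines and radial segment gradients, shaped by SmoothSteps and inverted, combine into arch and glow masks.
# Percentage drives the progress arch's segment spacing, while horizontal and vertical parameters control glow softness.
# EnableGlow blends between variants, and WaterGlow blends in a multiplied variant alongside a linear gradient mask and GlowColor parameter.
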Material /ColdDayUI/Materials/M_UI_Colday_CooldownGlow
graph {
 properties {blendMode=BLEND_Additive, materialDomain=MD_UI, pixelDepthOffsetMode=PDOM_Legacy, usedShadingModels=Default Lit}
 nodes [
  n1 = LinearInterpolate(A=NamedRerouteUsage(), B=NamedRerouteUsage(), Alpha=ScalarParameter() {defaultValue=1, parameterName=EnableGlow})
  n2 = NamedRerouteUsage()
  Multiply(A=LinearInterpolate(A=@n1, B=Multiply(A=@n2, B=@n1), Alpha=ScalarParameter() {parameterName=WaterGlow})) {constB=0.3}
  NamedRerouteDeclaration(None=SmoothStep(Value=MaterialFunctionCall() {materialFunction=&LinearGradient}.UGradient) {constMin=0.5}) {name=Mask, nodeColor=<rgba 0.5938 0 1 1>}
  n3 = NamedRerouteUsage()
  n4 = MaterialFunctionCall(UVs=NamedRerouteUsage(), Size=ScalarParameter() {defaultValue=0.87, parameterName=BracketsSize}, OutlineThickness=Constant() {r=0.01}) {materialFunction=&MF_UI_Circle}
  n5 = SmoothStep(Value=@n4.SDFOutline) {constMax=0.02}
  n6 = MaterialFunctionCall!(
   UV
   Segment Count=Constant() {r=1}
   Segment Spacing=LinearInterpolate(Alpha=ScalarParameter() {defaultValue=1, parameterName=Percentage}) {constA=0.36, constB=0.62}
   Fixed Spacing
  ) {materialFunction=/FortUILibrary/MaterialFunctions/Basics/MF_UI_RadialSegments}
  n7 = SmoothStep(Value=@n6.Segmented Triangles Gradient) {constMax=0.01}
  n8 = OneMinus(None=Multiply(A=OneMinus(None=@n5), B=OneMinus(None=@n7)))
  n9 = NamedRerouteUsage()
  n10 = SmoothStep(Max=Clamp(None=ScalarParameter() {parameterName=Cooldown_HorizontalGlow}) {maxDefault=0.1}, Value=@n4.SDFOutline) {constMax=0.106}
  n11 = SmoothStep(Max=Clamp(None=ScalarParameter() {parameterName=Cooldown_VerticalGlow}) {maxDefault=0.15}, Value=@n6.Segmented Triangles Gradient) {constMax=0.09}
  n12 = OneMinus(None=Multiply(A=OneMinus(None=@n10), B=OneMinus(None=@n11)))
  NamedRerouteDeclaration(None=LinearInterpolate(A=Subtract(A=@n3, B=@n8), B=LinearInterpolate(A=@n9, B=@n12, Alpha=@n9), Alpha=ScalarParameter() {defaultValue=1, parameterName=EnableGlow})) {name=CooldownProgressArch, nodeColor=<rgba 1 0.8438 0 1>}
  NamedRerouteDeclaration(None=TextureCoordinate()) {name=UV, nodeColor=<rgba 0 0 0 1>}
  n13 = MaterialFunctionCall(UVs=NamedRerouteUsage(), Size=Constant() {r=0.87}, OutlineThickness=Constant() {r=0.01}) {materialFunction=&MF_UI_Circle}
  n14 = SmoothStep(Max=Clamp(None=ScalarParameter() {parameterName=Cooldown_HorizontalGlow}) {maxDefault=0.1}, Value=@n13.SDFOutline) {constMax=0.106}
  n15 = MaterialFunctionCall(Segment Count=Constant() {r=2}, Segment Spacing=ScalarParameter() {defaultValue=0.24, parameterName=SegmentSpacing}) {materialFunction=&MF_UI_RadialSegments}
  n16 = SmoothStep(Max=Clamp(None=ScalarParameter() {parameterName=Cooldown_VerticalGlow}) {maxDefault=0.15}, Value=@n15.Segmented Triangles Gradient) {constMax=0.09}
  NamedRerouteDeclaration(None=Multiply(A=OneMinus(None=@n14), B=OneMinus(None=@n16))) {name=Cooldown_Glow, nodeColor=<rgba 1 0.2109 0 1>}
  n17 = SmoothStep(Value=@n13.SDFOutline) {constMax=0.426}
  n18 = SmoothStep(Value=@n15.Segmented Triangles Gradient) {constMax=0.338}
  NamedRerouteDeclaration(None=Multiply(A=OneMinus(None=@n17), B=OneMinus(None=@n18))) {name=CooldownArches, nodeColor=<rgba 0.9453 1 0 1>}
  VectorParameter() {parameterName=GlowColor, defaultValue=<rgba 1 1 1 1>}
 ]
}
