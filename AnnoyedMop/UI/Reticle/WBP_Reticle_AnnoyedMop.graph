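# Reticle widget for the AnnoyedMop weapon that shows target range, fire slam cooldown and combo state.
# Looping timers and a squared-distance check against a curve-table max range drive a reversible in-range bracket animation.
# Async waits on the cooldown gameplay tag switch cooldown/replenished animations with sound and manage the timer for indicator material progress.
# A retriggerable delay reverses the combo bracket animation when no combo is active.
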WidgetBlueprint /AnnoyedMop/UI/Reticle/WBP_Reticle_AnnoyedMop
animations [
 Anim_Intro {
  binding_count=1
  length_seconds=0.3
  object=&self:Anim_Intro
  sequence {bindings [Reticle_Root {
     id=#1
     possessed_class=Overlay
     tracks [
      {
       type=MovieSceneFloatTrack
       display_name=Render Opacity
       sections [{type=MovieSceneFloatSection, channels [None {type=MovieSceneScriptingFloatChannel, keys [[0,0],[6,1]]}]}]
      }
      {
       type=MovieScene2DTransformTrack
       sections [{
         type=MovieScene2DTransformSection
         channels [
          Translation.X {type=MovieSceneScriptingFloatChannel}
          Translation.Y {type=MovieSceneScriptingFloatChannel}
          Angle {type=MovieSceneScriptingFloatChannel}
          Scale.X {type=MovieSceneScriptingFloatChannel, keys [[0,1.25],[6,1]]}
          Scale.Y {type=MovieSceneScriptingFloatChannel, keys [[0,1.25],[6,1]]}
          Shear.X {type=MovieSceneScriptingFloatChannel}
          Shear.Y {type=MovieSceneScriptingFloatChannel}
         ]
        }]
      }
     ]
    }]}
 }
 Anim_ComboExpanding {
  binding_count=1
  length_seconds=1
  object=&self:Anim_ComboExpanding
  sequence {bindings [Image_MainBrackets {
     id=#2
     possessed_class=Image
     tracks [{
       type=MovieSceneWidgetMaterialTrack
       sections [{
         type=MovieSceneParameterSection
         end_frame=20
         start_frame=0
         channels [MiddleOfReticleXScale {type=MovieSceneScriptingFloatChannel, thinned_from=5, keys [[0,-0.4],[20,-0.2]]}]
        }]
      }]
    }]}
 }
 Anim_OnCooldown {
  binding_count=1
  length_seconds=0.25
  object=&self:Anim_OnCooldown
  sequence {bindings [Image_AmmoIndicator {
     id=#3
     possessed_class=Image
     tracks [{
       type=MovieSceneWidgetMaterialTrack
       sections [{
         type=MovieSceneParameterSection
         end_frame=5
         start_frame=0
         channels [
          ProgressFillAlpha {type=MovieSceneScriptingFloatChannel, keys [[0,1],[5,0.8]]}
          ProgressUnfilledAlpha {type=MovieSceneScriptingFloatChannel, const [1,2]}
          IsCooldownInProgress {type=MovieSceneScriptingFloatChannel, keys [[0,0],[5,0.75]]}
         ]
        }]
      }]
    }]}
 }
 Anim_OnReplenished {
  binding_count=2
  length_seconds=0.6079
  object=&self:Anim_OnReplenished
  sequence {bindings [
    Image_AmmoIndicator {
     id=#4
     possessed_class=Image
     tracks [{
       type=MovieSceneWidgetMaterialTrack
       sections [{
         type=MovieSceneParameterSection
         end_frame=12
         start_frame=0
         channels [
          ProgressFillAlpha {type=MovieSceneScriptingFloatChannel, thinned_from=4, keys [[0,0.75],[8.0003,1],[12,1]]}
          ProgressUnfilledAlpha {type=MovieSceneScriptingFloatChannel, thinned_from=4, keys [[0,0.5],[8.0003,1],[12,1]]}
          IsCooldownInProgress {type=MovieSceneScriptingFloatChannel, keys [[0,1],[2,0]]}
          CooldownBracketsFillColorR {type=MovieSceneScriptingFloatChannel, keys [[0,1],[12,0.7157]]}
          CooldownBracketsFillColorG {type=MovieSceneScriptingFloatChannel, keys [[0,1],[12,0.0823]]}
          CooldownBracketsFillColorB {type=MovieSceneScriptingFloatChannel, keys [[0,1],[12,0.0704]]}
          CooldownBracketsFillColorA {type=MovieSceneScriptingFloatChannel, const [1,2]}
         ]
        }]
      }]
    }
    Image_CooldownGlow {
     id=#5
     possessed_class=Image
     tracks [
      {
       type=MovieSceneWidgetMaterialTrack
       sections [{
         type=MovieSceneParameterSection
         end_frame=12
         start_frame=0
         channels [
          Cooldown_HorizontalGlow {type=MovieSceneScriptingFloatChannel, keys [[0,0],[3,0.065],[12.157,0]]}
          Cooldown_VerticalGlow {type=MovieSceneScriptingFloatChannel, keys [[0,0],[3,1.1073],[12.157,0]]}
          SegmentSpacing {type=MovieSceneScriptingFloatChannel, keys [[0,0],[8,0.2],[12.157,0]]}
         ]
        }]
      }
      {
       type=MovieSceneColorTrack
       display_name=Color and Opacity
       sections [{
         type=MovieSceneColorSection
         end_frame=12
         start_frame=0
         channels [
          Color.R {type=MovieSceneScriptingFloatChannel, const [0.7157,1]}
          Color.G {type=MovieSceneScriptingFloatChannel, const [0.0823,1]}
          Color.B {type=MovieSceneScriptingFloatChannel, const [0.0704,1]}
          Color.A {type=MovieSceneScriptingFloatChannel, keys [[0,2.2715],[12.1383,0]]}
         ]
        }]
      }
     ]
    }
   ]}
 }
 Anim_TargetInRange {
  binding_count=1
  length_seconds=0.1507
  object=&self:Anim_TargetInRange
  sequence {bindings [Image_MainBrackets {
     id=#6
     possessed_class=Image
     tracks [{
       type=MovieSceneWidgetMaterialTrack
       sections [{
         type=MovieSceneParameterSection
         end_frame=3
         start_frame=-1
         channels [
          CenterOpacity {type=MovieSceneScriptingFloatChannel, keys [[-0.0017,0.3],[3.0073,0.6]]}
          FireBendingCenter_Anim {type=MovieSceneScriptingFloatChannel, keys [[-0.0017,0],[3.0073,1]]}
         ]
        }]
      }]
    }]}
 }
]
graphs [
 AttemptToGetItemReference {nodes [
   AttemptToGetItemReference() {is=title K2Node_FunctionEntry}
   > n1 = Cast To FortPlayerController(Object=GetOwningPlayer())
   > Set CachedFortController(CachedFortController:Fort Player Controller Object Reference=@n1) -> (Output_Get:Fort Player Controller Object Reference)
   n2 = GetPlayerPawn!(self:Fort Player Controller Object Reference=@n1) -> (ReturnValue:Fort Player Pawn Object Reference)
   > Is Valid(InputObject=@n2)
   n3 = GetCurrentWeapon!(self:Fort Pawn Object Reference=@n2) -> (ReturnValue:Fort Weapon Object Reference)
   > Is Valid(InputObject=@n3)
   > Set CachedFortItem(CachedFortItem:Fort Item Object Reference=GetInventoryItem!(self:Fort Weapon Object Reference=@n3) -> (ReturnValue:Fort Item Object Reference)) -> (Output_Get:Fort Item Object Reference)
  ]}
 EventGraph {nodes [
   Event Construct()
   > n1 = Cast To B_Weapon_AnnoyedMop!(
    execute
    Object=GetEquippedWeapon!(self:Fort Weapon Reticle Extension Widget Base Object Reference) -> (ReturnValue:Fort Weapon Object Reference)
   ) -> (then, CastFailed, AsB Weapon Annoyed Mop:B Weapon Annoyed Mop Object Reference) {is=title K2Node_DynamicCast}
   > Set AnnoyedMopRef(AnnoyedMopRef:B Weapon Annoyed Mop Object Reference=@n1) -> (Output_Get:B Weapon Annoyed Mop Object Reference)
   > PlayAnimation(InAnimation=Get Anim_Intro() -> (Anim_Intro:Widget Animation Object Reference))
   > UpdateTargetInRangeState()
   > Set Timer by Event(Delegate=Create Event(), Time=0.1, bLooping=true)
   > AttemptToGetItemReference() {is=title K2Node_CallFunction}
   > n2 = WaitGameplayTagAddToActor!(
    execute:Exec
    TargetActor:Actor Object Reference=GetOwningPlayerPawn!(self:User Widget Object Reference) -> (ReturnValue:Pawn Object Reference)
    Tag:Gameplay Tag Structure=(TagName="Granted.Cooldown.AnnoyedMop")
    OnlyTriggerOnce:bool=false
   ) -> (then:Exec, AsyncTaskProxy:Ability Async Wait Gameplay Tag Added Object Reference, Added:Exec) {is=title K2Node_AsyncAction}
   > n3 = WaitGameplayTagRemoveFromActor!(
    execute:Exec
    TargetActor:Actor Object Reference=GetOwningPlayerPawn()
    Tag:Gameplay Tag Structure=(TagName="Granted.Cooldown.AnnoyedMop")
    OnlyTriggerOnce:bool=false
   ) -> (then:Exec, AsyncTaskProxy:Ability Async Wait Gameplay Tag Removed Object Reference, Removed:Exec) {is=title K2Node_AsyncAction}
   n4 = Branch(execute=@n3.Removed, Condition=IsAnimationPlaying(InAnimation=Get Anim_Intro()))
   PlayCooldownAnimation!(execute=@n4.else, self, Is In Cooldown:bool=false, SkipAnimation:bool=false) -> (then)
   > Clear and Invalidate Timer by Handle(Handle=Get FireSlamTimerHandle() -> (FireSlamTimerHandle:Timer Handle Structure))
   PlayCooldownAnimation(execute=@n2.Added, Is In Cooldown=true, SkipAnimation=false)
   > n5 = Set Timer by Event(Delegate=Create Event(), Time=0.05, bLooping=true)
   > Set FireSlamTimerHandle(FireSlamTimerHandle:Timer Handle Structure=@n5) -> (Output_Get:Timer Handle Structure)
  ]}
 PlayCooldownAnimation {nodes [
   n1 = PlayCooldownAnimation!() -> (then, Is In Cooldown:bool=~, SkipAnimation:bool=~) {is=title K2Node_FunctionEntry}
   n2 = Select(
    Option 0:Widget Animation Object Reference=Get Anim_OnReplenished() -> (Anim_OnReplenished:Widget Animation Object Reference)
    Option 1:Widget Animation Object Reference=Get Anim_OnCooldown() -> (Anim_OnCooldown:Widget Animation Object Reference)
    Index=@n1
   ) -> (ReturnValue:Widget Animation Object Reference)
   > PlayAnimation(InAnimation=@n2, StartAtTime=SelectFloat(A=GetEndTime(self=@n2), B=0, bPickA=@n1.SkipAnimation))
   > PlaySound2D(Sound=/AnnoyedMop/Assets/Audio/AnnoyedMop_UI_Cooldown_Over_Cue)
  ]}
 Sequencer Events {nodes [
   RetriggerableDelay(Duration=float * float(A=0.35, B:int=Get LastComboMove() -> (LastComboMove:int=0)) {is=title K2Node_PromotableOperator})
   > n1 = Branch(Condition=Get IsPerformingCombo() -> (IsPerformingCombo:bool))
   n2 = Get Anim_ComboExpanding() -> (Anim_ComboExpanding:Widget Animation Object Reference)
   n3 = GetEndTime(self=@n2)
   PlayAnimationTimeRange!(
    execute=@n1.else
    self:User Widget Object Reference
    InAnimation:Widget Animation Object Reference=@n2
    StartAtTime:float=@n3
    EndAtTime:float=float * float(A=SafeDivide(A=@n3, B=4), B=Get LastComboMove() -> (LastComboMove=0))
    NumLoopsToPlay:int=1
    PlayMode:EUMGSequencePlayMode Enum=Reverse
    PlaybackSpeed:float=1
    bRestoreState:bool=false
   ) -> (then, ReturnValue:Widget Animation Handle Structure)
  ]}
 UpdateFireSlamCooldown {nodes [
   UpdateFireSlamCooldown() {is=title K2Node_FunctionEntry}
   > n1 = GetDynamicMaterial(self=Get Image_AmmoIndicator() -> (Image_AmmoIndicator:Image Object Reference))
   n2 = GetCoolDownTimeRemaining!(
    self:Fort Item Object Reference=Get CachedFortItem() -> (CachedFortItem:Fort Item Object Reference)
    FortPC:Fort Player Controller Object Reference=Get CachedFortController() -> (CachedFortController:Fort Player Controller Object Reference)
   ) -> (CooldownDuration:float, ReturnValue:float)
   n3 = SafeDivide(A=float - float(A=@n2, B=@n2.ReturnValue), B=@n2)
   > SetScalarParameterValue(self=@n1, ParameterName=Progress, Value=SelectFloat(A=1, B=@n3, bPickA="Equal (Float)"(A=@n3)))
  ]}
 UpdateTargetInRangeState {nodes [
   UpdateTargetInRangeState() {is=title K2Node_FunctionEntry}
   > n1 = Cast To FortPlayerController(Object=GetOwningPlayer())
   n2 = GetLocationUnderReticle!(self:Fort Player Controller Object Reference=@n1) -> (ReturnValue:Vector)
   n3 = Get Actor Location(self=GetOwningPlayerPawn())
   n4 = GetValueAtLevel!(Input:"Scalable Float Structure (by ref)"=Get SF_UI_MaxRange() -> (SF_UI_MaxRange:Scalable Float Structure), Level:float=0) -> (ReturnValue:float)
   > n5 = Branch(Condition="float < float"(A="Distance Squared (Vector)"!(V1:Vector=@n2, V2:Vector=@n3) -> (ReturnValue:double), B=Square!(A:double=@n4) -> (ReturnValue:double)))
   > n6 = Branch(Condition=Get PlayedAnimInRange() -> (PlayedAnimInRange:bool))
   PlayAnimation(execute=@n6.else, InAnimation=Get Anim_TargetInRange() -> (Anim_TargetInRange:Widget Animation Object Reference))
   > Set PlayedAnimInRange(PlayedAnimInRange:bool=true) -> (Output_Get:bool)
   Branch(execute=@n5.else, Condition=Get PlayedAnimInRange())
   > PlayAnimation(InAnimation=Get Anim_TargetInRange(), PlayMode=Reverse)
   > Set PlayedAnimInRange(PlayedAnimInRange=false)
  ]}
]
variables [
 AnnoyedMopRef {type=B_Weapon_AnnoyedMop_C, default=null}
 CachedFortController {type=FortPlayerController, default=null}
 CachedFortItem {type=FortItem, default=null}
 FireSlamTimerHandle {type=TimerHandle}
 IsPerformingCombo {type=boolean}
 LastComboMove {type=integer}
 PlayedAnimInRange {type=boolean}
 SF_UI_MaxRange {
  type=ScalableFloat
  default {registryType=None, value=1, curve {curveTable=/AnnoyedMop/DataTables/AnnoyedMopGameData, rowName=Default.AnnoyedMop.UI.MaxRangeDisplay}}
 }
]
widgets {
 info {
  inheritedWidgetCount=1
  parentClass=/Script/FortniteUI.FortWeaponReticleExtensionWidgetBase
  rootWidgetClass=&Overlay
  widgetCount=5
 }
 widgets [
  Reticle_Root {widgetClassPath=&Overlay, widget=Reticle_Root}
  Image_MainBrackets {
   widgetClassPath=&Image
   bIsVariable=true
   parent=Reticle_Root
   widget=Image_MainBrackets
   properties {
    renderOpacity=0.9
    visibility=SelfHitTestInvisible
    brush {
     bIsDynamicallyLoaded=false
     drawAs=Image
     imageType=FullColor
     mirroring=NoMirror
     resourceName=None
     resourceObject=/AnnoyedMop/UI/Reticle/Material/MI_AnnoyedMop_Brackets
     tiling=NoTile
     imageSize=<120 120>
     margin=<ltrb 0 0 0 0>
     outlineSettings {
      bUseBrushTransparency=false
      roundingType=HalfHeightRadius
      width=0
      color {colorUseRule=UseColor_Specified, specifiedColor=<rgba 0 0 0 0>}
      cornerRadii=<0 0 0 0>
     }
     tintColor {colorUseRule=UseColor_Specified, specifiedColor=<rgba 1 1 1 1>}
     uVRegion {bIsValid=false, max=<0 0>, min=<0 0>}
    }
    renderTransform {angle=0, scale=<1.25 1.25>, shear=<0 0>, translation=<0 0>}
   }
   slot {class=OverlaySlot, properties {horizontalAlignment=HAlign_Center, verticalAlignment=VAlign_Center}}
  }
  Image_AmmoIndicator {
   widgetClassPath=&Image
   bIsVariable=true
   parent=Reticle_Root
   widget=Image_AmmoIndicator
   properties {
    visibility=SelfHitTestInvisible
    brush {
     bIsDynamicallyLoaded=false
     drawAs=Image
     imageType=FullColor
     mirroring=NoMirror
     resourceName=None
     resourceObject=/AnnoyedMop/UI/Reticle/Material/MI_UI_AnnoyedMop_CooldownIndicator
     tiling=NoTile
     imageSize=<120 120>
     margin=<ltrb 0 0 0 0>
     outlineSettings {
      bUseBrushTransparency=false
      roundingType=HalfHeightRadius
      width=0
      color {colorUseRule=UseColor_Specified, specifiedColor=<rgba 0 0 0 0>}
      cornerRadii=<0 0 0 0>
     }
     tintColor {colorUseRule=UseColor_Specified, specifiedColor=<rgba 1 1 1 1>}
     uVRegion {bIsValid=false, max=<0 0>, min=<0 0>}
    }
    renderTransform {angle=0, scale=<1.25 1.25>, shear=<0 0>, translation=<0 0>}
   }
   slot {class=OverlaySlot, properties {horizontalAlignment=HAlign_Center, verticalAlignment=VAlign_Center}}
  }
  Image_CooldownGlow {
   widgetClassPath=&Image
   bIsVariable=true
   parent=Reticle_Root
   widget=Image_CooldownGlow
   properties {
    brush {
     bIsDynamicallyLoaded=false
     drawAs=Image
     imageType=FullColor
     mirroring=NoMirror
     resourceName=None
     resourceObject=/ColdDayUI/Materials/M_UI_Colday_CooldownGlow
     tiling=NoTile
     imageSize=<32 32>
     margin=<ltrb 0 0 0 0>
     outlineSettings {
      bUseBrushTransparency=false
      roundingType=HalfHeightRadius
      width=0
      color {colorUseRule=UseColor_Specified, specifiedColor=<rgba 0 0 0 0>}
      cornerRadii=<0 0 0 0>
     }
     tintColor {colorUseRule=UseColor_Specified, specifiedColor=<rgba 1 1 1 1>}
     uVRegion {bIsValid=false, max=<0 0>, min=<0 0>}
    }
    colorAndOpacity=<rgba 1 1 1 0>
    renderTransform {angle=0, scale=<1.22 1.25>, shear=<0 0>, translation=<0 0>}
   }
   slot {class=OverlaySlot, properties {horizontalAlignment=HAlign_Fill, verticalAlignment=VAlign_Fill}}
  }
  TutorialHighlight {widgetClassPath=&TutorialHighlight}
 ]
}
